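# Physics asset for the ASN_2 hair part of the F_MED_Elastic female body, giving the hair constrained secondary motion.
# Capsule bodies cover the head, a hair root and the front, side and top strands.
# Side, top and two front base bodies hang off the root with tight angular limits, and each front base branches into two children, some twist-locked.
# The root joins the head with free swing and twist.
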PhysicsAsset /BRCosmetics/Characters/Player/Female/Medium/Bodies/F_MED_Elastic/Meshes/Parts/ASN_Parts/F_MED_Elastic_ASN_2_Physics
bodies {
 asn2_fr_hair_A_a [{shapeType=Capsule, length=0.75, radius=1.25, center=<1.5965 0.0001 0>, rotation=<rot 90 0 0>}]
 asn2_fr_hair_A_b [{shapeType=Capsule, length=2.1412, radius=1.4355, center=<0.9591 0.1203 -0.2464>, rotation=<rot 0 91.1999 -78.3413>}]
 asn2_fr_hair_A_c [{shapeType=Capsule, length=1.5967, radius=1.0489, center=<0.794 -0.1749 0.0492>, rotation=<rot 0 -80.7857 -88.6951>}]
 asn2_fr_hair_B_a [{shapeType=Capsule, length=2.75, radius=1.25, center=<1.9804 0.0001 0>, rotation=<rot 90 0 0>}]
 asn2_fr_hair_B_b [{shapeType=Capsule, length=1.0718, radius=1.6118, center=<1.3446 -0.0174 -0.129>, rotation=<rot 0 97.0176 -83.9674>}]
 asn2_fr_hair_B_c [{shapeType=Capsule, length=3.2369, radius=1.1684, center=<2.5325 -0.0494 -0.1331>, rotation=<rot 0 89.5662 -88.0128>}]
 asn2_hair [{shapeType=Capsule, length=1, radius=1}]
 asn2_side_hair_a_l [{shapeType=Capsule, length=1.255, radius=0.755, center=<0.9143 0.5099 -0.3878>, rotation=<rot 0 -76.5725 -88.7202>}]
 asn2_side_hair_a_r [{shapeType=Capsule, length=1.1001, radius=0.8237, center=<0.3652 -0.3227 -0.6657>, rotation=<rot 25.7754 -90 -25.0024>}]
 asn2_side_hair_b_r [{shapeType=Capsule, length=1.8824, radius=1.0765, center=<0.0049 -0.4908 -1.6295>, rotation=<rot 10.0821 -91.4299 7.8262>}]
 asn2_side_hair_c_r [{shapeType=Capsule, length=1.2744, radius=1.1682, center=<-0.0946 -0.0622 -1.279>, rotation=<rot 8.7219 -90 14.2077>}]
 asn2_side_hair_d_r [{shapeType=Capsule, length=1.4796, radius=0.7584, center=<0.7615 -0.148 -1.0308>, rotation=<rot 10.7176 -90 -19.9556>}]
 asn2_side_hair_e_r [{shapeType=Capsule, length=2.434, radius=0.9964, center=<0.9785 -0.5398 -1.3949>, rotation=<rot 25.4641 -90 -28.5427>}]
 asn2_top_hair_a_r [{shapeType=Capsule, length=1.7356, radius=1.0577, center=<-0.5357 -0.247 -1.2478>, rotation=<rot 14.2741 -89.9999 24.5843>}]
 asn2_top_hair_b_r [{shapeType=Capsule, length=1.4464, radius=1.12, center=<-1.2311 -0.0482 -0.7055>, rotation=<rot -11.9958 94.1437 -68.6377>}]
 asn2_top_hair_c_r [{shapeType=Capsule, length=1.4841, radius=1.1792, center=<-0.5041 -0.5606 -0.6332>, rotation=<rot 14.276 140.7053 -32.6422>}]
 head [{shapeType=Capsule, length=5.25, radius=6.5, center=<4.3329 2.4357 0>, rotation=<rot 90 0 0>}]
}
constraints [
 {bone1Name=asn2_side_hair_a_l, bone2Name=asn2_hair, swing1LimitDegrees=15, swing2LimitDegrees=15, twistLimitDegrees=15}
 {bone1Name=asn2_top_hair_a_r, bone2Name=asn2_hair, swing1LimitDegrees=10, swing2LimitDegrees=10, twistLimitDegrees=10}
 {bone1Name=asn2_side_hair_b_r, bone2Name=asn2_hair, swing1LimitDegrees=15, swing2LimitDegrees=15, twistLimitDegrees=15}
 {bone1Name=asn2_side_hair_c_r, bone2Name=asn2_hair, swing1LimitDegrees=10, swing2LimitDegrees=10, twistLimitDegrees=10}
 {bone1Name=asn2_top_hair_b_r, bone2Name=asn2_hair, swing1LimitDegrees=10, swing2LimitDegrees=10, twistLimitDegrees=10}
 {bone1Name=asn2_top_hair_c_r, bone2Name=asn2_hair, swing1LimitDegrees=10, swing2LimitDegrees=10, twistLimitDegrees=10}
 {bone1Name=asn2_side_hair_e_r, bone2Name=asn2_hair, swing1LimitDegrees=15, swing2LimitDegrees=15, twistLimitDegrees=15}
 {bone1Name=asn2_side_hair_d_r, bone2Name=asn2_hair, swing1LimitDegrees=15, swing2LimitDegrees=15, twistLimitDegrees=15}
 {bone1Name=asn2_side_hair_a_r, bone2Name=asn2_hair, swing1LimitDegrees=15, swing2LimitDegrees=15, twistLimitDegrees=15}
 {bone1Name=asn2_fr_hair_B_a, bone2Name=asn2_hair, swing1LimitDegrees=5, swing2LimitDegrees=5, twistLimitDegrees=5, twistMotion=Locked}
 {
  bone1Name=asn2_fr_hair_B_c
  bone2Name=asn2_fr_hair_B_a
  swing1LimitDegrees=10
  swing2LimitDegrees=10
  twistLimitDegrees=5
  twistMotion=Locked
 }
 {
  bone1Name=asn2_fr_hair_B_b
  bone2Name=asn2_fr_hair_B_a
  swing1LimitDegrees=10
  swing2LimitDegrees=10
  twistLimitDegrees=15
  twistMotion=Locked
 }
 {bone1Name=asn2_fr_hair_A_a, bone2Name=asn2_hair, swing1LimitDegrees=5, swing2LimitDegrees=5, twistLimitDegrees=5, twistMotion=Locked}
 {bone1Name=asn2_fr_hair_A_b, bone2Name=asn2_fr_hair_A_a, swing1LimitDegrees=10, swing2LimitDegrees=10, twistLimitDegrees=10}
 {bone1Name=asn2_fr_hair_A_c, bone2Name=asn2_fr_hair_A_a, swing1LimitDegrees=10, swing2LimitDegrees=10, twistLimitDegrees=10}
 {
  bone1Name=asn2_hair
  bone2Name=head
  swing1LimitDegrees=45
  swing1Motion=Free
  swing2LimitDegrees=45
  swing2Motion=Free
  twistLimitDegrees=45
  twistMotion=Free
 }
]
