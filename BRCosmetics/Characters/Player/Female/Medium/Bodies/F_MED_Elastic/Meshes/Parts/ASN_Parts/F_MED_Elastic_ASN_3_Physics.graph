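# Physics asset for the ASN_3 hair part of the medium female Elastic player body, covering twin ponytails and side hair strands.
# Capsule bodies on head, neck, upper spine, clavicles and upper arms, plus a large thin tilted box on spine_05, provide upper-body shapes around the hair.
# Body joints and hair roots rotate freely.
# Ponytail chains branch into detail strands with moderate swing and tight twist limits, while asymmetric side-hair chains are held stiffer.
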PhysicsAsset /BRCosmetics/Characters/Player/Female/Medium/Bodies/F_MED_Elastic/Meshes/Parts/ASN_Parts/F_MED_Elastic_ASN_3_Physics
bodies {
 ans3_sideHair_A_a_l [{shapeType=Capsule, length=2.25, radius=1, center=<2.0564 0 0>, rotation=<rot 90 0 0>}]
 ans3_sideHair_A_a_r [{shapeType=Capsule, length=2, radius=1.5, center=<1.9341 0 0>, rotation=<rot 90 0 0>}]
 ans3_sideHair_A_b_l [{shapeType=Capsule, length=3.75, radius=0.75, center=<2.3692 0 0>, rotation=<rot 90 0 0>}]
 ans3_sideHair_A_b_r [{shapeType=Capsule, length=1.5, radius=1.25, center=<2.0616 0.0001 0>, rotation=<rot 90 0 0>}]
 ans3_sideHair_A_c_r [{shapeType=Capsule, length=1.25, radius=1, center=<1.5877 0 0>, rotation=<rot 90 0 0>}]
 ans3_sideHair_A_d_r [{shapeType=Capsule, length=1.75, radius=1, center=<1.5057 0 0>, rotation=<rot 90 0 0>}]
 asn3_hair [{shapeType=Capsule, length=1, radius=1}]
 asn3_ponytail_a_base_l [{shapeType=Capsule, length=1, radius=1}]
 asn3_ponytail_a_base_r [{shapeType=Capsule, length=1, radius=1}]
 asn3_ponytail_a_l [{shapeType=Capsule, length=2.5, radius=2.25, center=<3.2776 0 0>, rotation=<rot 90 0 0>}]
 asn3_ponytail_a_r [{shapeType=Capsule, length=2.5, radius=2.25, center=<-3.2776 0 0>, rotation=<rot 90 0 180>}]
 asn3_ponytail_b_l [{shapeType=Capsule, length=1.25, radius=3.25, center=<3.1159 0 0.2592>, rotation=<rot 90 0 0>}]
 asn3_ponytail_b_r [{shapeType=Capsule, length=1.75, radius=2.75, center=<-3.1159 -0.0215 -0.2592>, rotation=<rot 90 0 -180>}]
 asn3_ponytail_c_l [{shapeType=Capsule, length=4.25, radius=1.25, center=<3.4385 -0.0001 0>, rotation=<rot 90 90 93.6551>}]
 asn3_ponytail_c_r [{shapeType=Capsule, length=3.25, radius=1.75, center=<-3.0732 -0.0233 0>, rotation=<rot 90 5.3558 -170.989>}]
 asn3_ponytail_detail_a_l [{shapeType=Capsule, length=1.5, radius=0.75, center=<1.1866 0 0>, rotation=<rot 90 0 0>}]
 asn3_ponytail_detail_a_r [{shapeType=Capsule, length=2.5725, radius=1.0281, center=<-1.685 0.0893 0.3441>, rotation=<rot 0 -85.0594 -78.3848>}]
 clavicle_l [{shapeType=Capsule, length=7.25, radius=5.25, center=<4.7457 -1.0563 -0.2627>, rotation=<rot 81.2983 0 0>}]
 clavicle_r [{shapeType=Capsule, length=7.25, radius=5.25, center=<-4.7457 1.0563 0.2627>, rotation=<rot 81.2983 0 -180>}]
 head [
  {shapeType=Capsule, length=3.75, radius=6.25, center=<2.5 3 0>, rotation=<rot 90 0 0>}
  {shapeType=Capsule, length=10, radius=1.75, shapeName=head_capsule, center=<-1.3373 -6.3255 0>, rotation=<rot 90 90 90.0002>}
 ]
 neck_01 [{shapeType=Capsule, length=3, radius=4.5, center=<1.9014 0.368 0>, rotation=<rot 90 0 0>}]
 spine_04 [{shapeType=Capsule, length=9, radius=9.5, center=<2.9073 2.0518 0>}]
 spine_05 [
  {shapeType=Capsule, length=12, radius=6.75, center=<4.3499 -0.7789 0>}
  {shapeType=Box, extentX=18.5, extentY=2.5, extentZ=36.5, center=<11.955 3.0055 0>, rotation=<rot 0 14.6917 0>}
 ]
 upperarm_l [{shapeType=Capsule, length=17.75, radius=4.5, center=<10.72 -0.1352 0>, rotation=<rot 90 0 2.6996>}]
 upperarm_r [{shapeType=Capsule, length=17.75, radius=4.5, center=<-10.72 0.1352 0>, rotation=<rot 90 0 -177.3004>}]
}
constraints [
 {
  bone1Name=head
  bone2Name=neck_01
  swing1LimitDegrees=45
  swing1Motion=Free
  swing2LimitDegrees=45
  swing2Motion=Free
  twistLimitDegrees=45
  twistMotion=Free
 }
 {
  bone1Name=upperarm_l
  bone2Name=clavicle_l
  swing1LimitDegrees=45
  swing1Motion=Free
  swing2LimitDegrees=45
  swing2Motion=Free
  twistLimitDegrees=45
  twistMotion=Free
 }
 {
  bone1Name=upperarm_r
  bone2Name=clavicle_r
  swing1LimitDegrees=45
  swing1Motion=Free
  swing2LimitDegrees=45
  swing2Motion=Free
  twistLimitDegrees=45
  twistMotion=Free
 }
 {
  bone1Name=asn3_hair
  bone2Name=head
  swing1LimitDegrees=45
  swing1Motion=Free
  swing2LimitDegrees=45
  swing2Motion=Free
  twistLimitDegrees=45
  twistMotion=Free
 }
 {
  bone1Name=asn3_ponytail_a_base_r
  bone2Name=asn3_hair
  swing1LimitDegrees=45
  swing1Motion=Free
  swing2LimitDegrees=45
  swing2Motion=Free
  twistLimitDegrees=45
  twistMotion=Free
 }
 {
  bone1Name=asn3_ponytail_a_base_l
  bone2Name=asn3_hair
  swing1LimitDegrees=45
  swing1Motion=Free
  swing2LimitDegrees=45
  swing2Motion=Free
  twistLimitDegrees=45
  twistMotion=Free
 }
 {bone1Name=asn3_ponytail_a_l, bone2Name=asn3_ponytail_a_base_l, swing1LimitDegrees=25, swing2LimitDegrees=25, twistLimitDegrees=5}
 {bone1Name=asn3_ponytail_b_l, bone2Name=asn3_ponytail_a_l, swing1LimitDegrees=25, swing2LimitDegrees=25, twistLimitDegrees=5}
 {bone1Name=asn3_ponytail_c_l, bone2Name=asn3_ponytail_b_l, swing1LimitDegrees=25, swing2LimitDegrees=25, twistLimitDegrees=5}
 {bone1Name=asn3_ponytail_detail_a_l, bone2Name=asn3_ponytail_b_l, swing1LimitDegrees=25, swing2LimitDegrees=25, twistLimitDegrees=5}
 {
  bone1Name=clavicle_l
  bone2Name=spine_05
  swing1LimitDegrees=45
  swing1Motion=Free
  swing2LimitDegrees=45
  swing2Motion=Free
  twistLimitDegrees=45
  twistMotion=Free
 }
 {
  bone1Name=clavicle_r
  bone2Name=spine_05
  swing1LimitDegrees=45
  swing1Motion=Free
  swing2LimitDegrees=45
  swing2Motion=Free
  twistLimitDegrees=45
  twistMotion=Free
 }
 {
  bone1Name=neck_01
  bone2Name=spine_05
  swing1LimitDegrees=45
  swing1Motion=Free
  swing2LimitDegrees=45
  swing2Motion=Free
  twistLimitDegrees=45
  twistMotion=Free
 }
 {
  bone1Name=spine_05
  bone2Name=spine_04
  swing1LimitDegrees=45
  swing1Motion=Free
  swing2LimitDegrees=45
  swing2Motion=Free
  twistLimitDegrees=45
  twistMotion=Free
 }
 {bone1Name=asn3_ponytail_a_r, bone2Name=asn3_ponytail_a_base_r, swing1LimitDegrees=25, swing2LimitDegrees=25, twistLimitDegrees=5}
 {bone1Name=asn3_ponytail_b_r, bone2Name=asn3_ponytail_a_r, swing1LimitDegrees=25, swing2LimitDegrees=25, twistLimitDegrees=5}
 {bone1Name=asn3_ponytail_c_r, bone2Name=asn3_ponytail_b_r, swing1LimitDegrees=25, swing2LimitDegrees=25, twistLimitDegrees=5}
 {bone1Name=asn3_ponytail_detail_a_r, bone2Name=asn3_ponytail_b_r, swing1LimitDegrees=15, swing2LimitDegrees=15, twistLimitDegrees=5}
 {bone1Name=ans3_sideHair_A_a_r, bone2Name=asn3_hair, swing1LimitDegrees=5, swing2LimitDegrees=10, twistLimitDegrees=5}
 {bone1Name=ans3_sideHair_A_b_r, bone2Name=ans3_sideHair_A_a_r, swing1LimitDegrees=5, swing2LimitDegrees=10, twistLimitDegrees=5}
 {bone1Name=ans3_sideHair_A_c_r, bone2Name=ans3_sideHair_A_b_r, swing1LimitDegrees=5, swing2LimitDegrees=10, twistLimitDegrees=5}
 {bone1Name=ans3_sideHair_A_d_r, bone2Name=ans3_sideHair_A_b_r, swing1LimitDegrees=5, swing2LimitDegrees=5, twistLimitDegrees=5}
 {bone1Name=ans3_sideHair_A_a_l, bone2Name=asn3_hair, swing1LimitDegrees=5, swing2LimitDegrees=5, twistLimitDegrees=5}
 {bone1Name=ans3_sideHair_A_b_l, bone2Name=ans3_sideHair_A_a_l, swing1LimitDegrees=5, swing2LimitDegrees=5, twistLimitDegrees=5}
]
