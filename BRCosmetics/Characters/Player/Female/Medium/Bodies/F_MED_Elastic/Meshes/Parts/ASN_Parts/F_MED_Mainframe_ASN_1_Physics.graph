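# Physics asset for the ASN_1 hair part of the female medium Mainframe character, providing bodies for hair physics.
# Capsules cover the upper spine, clavicles, neck and head, plus three main hair chains (A, B, C) and three smaller pieces.
# Body links and hair root joints swing and twist freely.
# Strand joints carry narrow angular limits, tighter on chain A than on B and C, with piece asn1_hair_a branching from chain A.
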PhysicsAsset /BRCosmetics/Characters/Player/Female/Medium/Bodies/F_MED_Elastic/Meshes/Parts/ASN_Parts/F_MED_Mainframe_ASN_1_Physics
bodies {
 asn1_hair [{shapeType=Capsule, length=1, radius=1}]
 asn1_hair_a [{shapeType=Capsule, length=2.25, radius=1, center=<1.8098 0 0>, rotation=<rot 90 0 0>}]
 asn1_hair_b [{shapeType=Capsule, length=1.75, radius=1, center=<1.5 0 0>, rotation=<rot 90 0 0>}]
 asn1_hair_c [{shapeType=Capsule, length=1.75, radius=1, center=<1.5 0 0>, rotation=<rot 90 0 0>}]
 asn1_hair_main_A_a [{shapeType=Capsule, length=3, radius=1.75, center=<3.105 0 0>, rotation=<rot 90 0 0>}]
 asn1_hair_main_A_a_base [{shapeType=Capsule, length=1, radius=1}]
 asn1_hair_main_A_b [{shapeType=Capsule, length=2.25, radius=2, center=<2.7376 0 0>, rotation=<rot 86.3848 0 0>}]
 asn1_hair_main_A_c [{shapeType=Capsule, length=2.25, radius=1.5, center=<2.022 0 0>, rotation=<rot 90 0 0>}]
 asn1_hair_main_A_d [{shapeType=Capsule, length=3.75, radius=1.25, center=<2.6934 0.3991 0>, rotation=<rot 90 0 -12.8959>}]
 asn1_hair_main_B_a [{shapeType=Capsule, length=2.5, radius=2.5, center=<3.0618 0 0>, rotation=<rot 90 0 0>}]
 asn1_hair_main_B_a_base [{shapeType=Capsule, length=1, radius=1}]
 asn1_hair_main_B_b [{shapeType=Capsule, length=1.75, radius=2.25, center=<2.5268 0 0>, rotation=<rot 90 0 0>}]
 asn1_hair_main_B_c [{shapeType=Capsule, length=0.75, radius=1.75, center=<1.7019 0 0>, rotation=<rot 90 0 0>}]
 asn1_hair_main_B_d [{shapeType=Capsule, length=1.75, radius=1, center=<1.5542 0.1311 0>, rotation=<rot 90 0 0>}]
 asn1_hair_main_C_a [{shapeType=Capsule, length=2.5, radius=2.25, center=<3.4616 0 0.0001>, rotation=<rot 90 0 0>}]
 asn1_hair_main_C_a_base [{shapeType=Capsule, length=1, radius=1}]
 asn1_hair_main_C_b [{shapeType=Capsule, length=1.75, radius=1.75, center=<2.2691 0.0001 0>, rotation=<rot 90 0 0>}]
 asn1_hair_main_C_c [{shapeType=Capsule, length=3.25, radius=1.25, center=<2.7361 0.0001 0>, rotation=<rot 90 0 0>}]
 clavicle_l [{shapeType=Capsule, length=8.25, radius=6.25, center=<4.6546 -1.1032 -1.3835>, rotation=<rot 84.4904 0 0>}]
 clavicle_r [{shapeType=Capsule, length=8.25, radius=6.25, center=<-4.6546 1.1032 1.3835>, rotation=<rot 84.4904 0 180>}]
 head [{shapeType=Capsule, length=5, radius=6.5, center=<2.4796 2.289 0.6971>, rotation=<rot 90 0 0>}]
 neck_01 [{shapeType=Capsule, length=4.75, radius=4.5, center=<2.166 0.4802 0>, rotation=<rot 90 0 0.8586>}]
 spine_04 [{shapeType=Capsule, length=3.5, radius=9.75, center=<3.4006 1.8235 0.0002>}]
 spine_05 [
  {shapeType=Capsule, length=5.2499, radius=8.6287, center=<3.084 -1.1687 -0.0001>}
  {shapeType=Capsule, length=8.6487, radius=8.6287, center=<-0.8742 -2.4078 -0.0001>}
 ]
}
constraints [
 {
  bone1Name=head
  bone2Name=neck_01
  swing1LimitDegrees=45
  swing1Motion=Free
  swing2LimitDegrees=45
  swing2Motion=Free
  twistLimitDegrees=45
  twistMotion=Free
 }
 {
  bone1Name=spine_05
  bone2Name=spine_04
  swing1LimitDegrees=45
  swing1Motion=Free
  swing2LimitDegrees=45
  swing2Motion=Free
  twistLimitDegrees=45
  twistMotion=Free
 }
 {
  bone1Name=neck_01
  bone2Name=spine_05
  swing1LimitDegrees=45
  swing1Motion=Free
  swing2LimitDegrees=45
  swing2Motion=Free
  twistLimitDegrees=45
  twistMotion=Free
 }
 {
  bone1Name=clavicle_l
  bone2Name=spine_05
  swing1LimitDegrees=45
  swing1Motion=Free
  swing2LimitDegrees=45
  swing2Motion=Free
  twistLimitDegrees=45
  twistMotion=Free
 }
 {
  bone1Name=asn1_hair
  bone2Name=head
  swing1LimitDegrees=45
  swing1Motion=Free
  swing2LimitDegrees=45
  swing2Motion=Free
  twistLimitDegrees=45
  twistMotion=Free
 }
 {
  bone1Name=asn1_hair_main_A_a_base
  bone2Name=asn1_hair
  swing1LimitDegrees=45
  swing1Motion=Free
  swing2LimitDegrees=45
  swing2Motion=Free
  twistLimitDegrees=45
  twistMotion=Free
 }
 {
  bone1Name=asn1_hair_main_B_a_base
  bone2Name=asn1_hair
  swing1LimitDegrees=45
  swing1Motion=Free
  swing2LimitDegrees=45
  swing2Motion=Free
  twistLimitDegrees=45
  twistMotion=Free
 }
 {
  bone1Name=asn1_hair_main_C_a_base
  bone2Name=asn1_hair
  swing1LimitDegrees=45
  swing1Motion=Free
  swing2LimitDegrees=45
  swing2Motion=Free
  twistLimitDegrees=45
  twistMotion=Free
 }
 {bone1Name=asn1_hair_main_A_a, bone2Name=asn1_hair_main_A_a_base, swing1LimitDegrees=15, swing2LimitDegrees=15, twistLimitDegrees=5}
 {bone1Name=asn1_hair_main_A_b, bone2Name=asn1_hair_main_A_a, swing1LimitDegrees=15, swing2LimitDegrees=15, twistLimitDegrees=5}
 {bone1Name=asn1_hair_main_A_c, bone2Name=asn1_hair_main_A_b, swing1LimitDegrees=15, swing2LimitDegrees=15, twistLimitDegrees=5}
 {bone1Name=asn1_hair_main_A_d, bone2Name=asn1_hair_main_A_c, swing1LimitDegrees=15, swing2LimitDegrees=15, twistLimitDegrees=5}
 {bone1Name=asn1_hair_a, bone2Name=asn1_hair_main_A_b, swing1LimitDegrees=30, swing2LimitDegrees=30, twistLimitDegrees=10}
 {bone1Name=asn1_hair_main_B_a, bone2Name=asn1_hair_main_B_a_base, swing1LimitDegrees=30, swing2LimitDegrees=30, twistLimitDegrees=10}
 {bone1Name=asn1_hair_main_B_b, bone2Name=asn1_hair_main_B_a, swing1LimitDegrees=30, swing2LimitDegrees=30, twistLimitDegrees=10}
 {bone1Name=asn1_hair_main_B_c, bone2Name=asn1_hair_main_B_b, swing1LimitDegrees=30, swing2LimitDegrees=30, twistLimitDegrees=10}
 {bone1Name=asn1_hair_main_B_d, bone2Name=asn1_hair_main_B_c, swing1LimitDegrees=30, swing2LimitDegrees=30, twistLimitDegrees=10}
 {
  bone1Name=clavicle_r
  bone2Name=spine_05
  swing1LimitDegrees=45
  swing1Motion=Free
  swing2LimitDegrees=45
  swing2Motion=Free
  twistLimitDegrees=45
  twistMotion=Free
 }
 {bone1Name=asn1_hair_main_C_a, bone2Name=asn1_hair_main_C_a_base, swing1LimitDegrees=30, swing2LimitDegrees=30, twistLimitDegrees=10}
 {bone1Name=asn1_hair_main_C_b, bone2Name=asn1_hair_main_C_a, swing1LimitDegrees=30, swing2LimitDegrees=30, twistLimitDegrees=10}
 {bone1Name=asn1_hair_main_C_c, bone2Name=asn1_hair_main_C_b, swing1LimitDegrees=30, swing2LimitDegrees=30, twistLimitDegrees=10}
 {bone1Name=asn1_hair_b, bone2Name=asn1_hair, swing1LimitDegrees=15, swing2LimitDegrees=15, twistLimitDegrees=15}
 {bone1Name=asn1_hair_c, bone2Name=asn1_hair, swing1LimitDegrees=15, swing2LimitDegrees=15, twistLimitDegrees=15}
]
